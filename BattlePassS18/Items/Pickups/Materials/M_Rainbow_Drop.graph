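# Subsurface material for the Rainbow Drop pickup that builds a glowing rainbow look with animated wobble on skeletal meshes.
# A rotated pre-skinned position gradient plus vertex red hue-shifts a start color, with optional banding via a static switch.
# Sunlight-vector toon shading, fake specular and fresnel are added to that color.
# A normal-direction offset sums world-space and local sine waves, faded by camera distance.
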Material /BattlePassS18/Items/Pickups/Materials/M_Rainbow_Drop
graph {
 properties {
  bUsedWithMeshDeformer=true
  bUsedWithSkeletalMesh=true
  pixelDepthOffsetMode=PDOM_Legacy
  shadingModel=MSM_Subsurface
  usedShadingModels=Subsurface
 }
 nodes [
  n1 = PixelNormalWS()
  n2 = Multiply(A=Fresnel(ExponentIn=Constant() {r=1}, Normal=@n1)) {constB=0.15}
  n3 = CollectionParameter() {
   collection=&FortniteMaterialParameters
   parameterName=SunlightDirectionalVector
  }
  n4 = ComponentMask(None=@n3) {b=true, g=true, r=true}
  n5 = Multiply(A=Multiply(A=DotProduct(A=@n4, B=@n1)) {constB=2}, B=@n1)
  n6 = DotProduct(
   A=Normalize(VectorInput=Subtract(A=@n5, B=@n4))
   B=Multiply(A=MaterialFunctionCall() {materialFunction=&CameraDirectionVector}.Result) {constB=-1}
  )
  n7 = MaterialFunctionCall(
   In=Power(Base=Max(A=@n6) {constB=0}, Exp=ScalarParameter() {defaultValue=10, group=0SlurpMonsterSummer, parameterName=FakeSpecPow})
   Contrast=ScalarParameter() {defaultValue=5, group=0SlurpMonsterSummer, parameterName=FakespecHardness}
  ) {materialFunction=&CheapContrast}
  n8 = CollectionParameter() {
   collection=&FortniteMaterialParameters
   parameterName=SunlightDirectionalVector
  }
  n9 = SmoothStep(
   Max=ScalarParameter() {defaultValue=0.1, parameterName=ShadingSmooth}
   Value=DotProduct(A=ComponentMask(None=@n8) {b=true, g=true, r=true}, B=PixelNormalWS())
  ) {constMax=0}
  n10 = PreSkinnedPosition()
  n11 = PreSkinnedLocalBounds()
  n12 = Divide(A=VertexInterpolator(VS=@n10).PS, B=Subtract(A=@n11.Max, B=@n11.Max))
  n13 = MaterialFunctionCall(
   UVs=ComponentMask(None=@n12) {b=true, r=true}
   Rotation Center=Constant() {r=0.5}
   "Rotation Angle (0-1)"=ScalarParameter() {defaultValue=0.026, parameterName=Rotation Amount}
  ) {materialFunction=&CustomRotator}
  n14 = Add(A=Multiply(A=ComponentMask(None=@n13.Rotated Values) {g=true}) {constB=2}, B=Multiply(A=VertexColor().R) {constB=1.5})
  n15 = ScalarParameter() {defaultValue=6, parameterName=Step}
  n16 = Multiply(A=@n14, B=@n15) {constB=10}
  n17 = Multiply(A=Divide(A=Floor(None=@n16), B=@n15)) {constB=5}
  n18 = Multiply(
   A=MaterialFunctionCall(
    Hue Shift Percentage=StaticSwitchParameter(True=@n17, False=@n14) {parameterName=Use Step}
    Texture=VectorParameter() {parameterName=Hue_Start, defaultValue=<rgba 1 0 0 1>}.RGB
   ) {materialFunction=&HueShift}.Result
   B=ScalarParameter() {defaultValue=0.6, parameterName=EmissiveMult}
  )
  Add(
   A=Multiply(A=Add(A=@n2, B=@n7.Result), B=ScalarParameter() {defaultValue=2, group=0SlurpMonsterSummer, parameterName=FakespecEmsMult})
   B=LinearInterpolate(A=Multiply(A=@n9, B=@n18), B=@n18) {constAlpha=0.6}
  )
  n19 = ComponentMask(None=@n12) {b=true}
  ConstantBiasScale(None=@n19) {bias=0.1, scale=1.6}
  Constant()
  Constant() {r=0.7}
  Constant() {r=0.001}
  Constant() {r=0.1}
  n20 = WorldPosition() {worldPositionShaderOffset=WPT_ExcludeAllShaderOffsets}
  n21 = Subtract(A=Frac(None=Multiply(A=@n20.XYZ) {constB=0.0025})) {constB=0.5}
  n22 = MaterialFunctionCall(Float3=Abs(None=Multiply(A=@n21) {constB=2})) {materialFunction=&BreakOutFloat3Components}
  n23 = Add(A=Add(A=@n22.R, B=@n22.R), B=@n22.B)
  n24 = Multiply(A=Time()) {constB=0.25}
  n25 = Sine(None=Subtract(A=Subtract(A=@n23) {constB=2}, B=Frac(None=@n24)))
  n26 = Multiply(A=Multiply(A=Add(A=@n25)) {constB=0.5}, B=ScalarParameter() {defaultValue=0.5, parameterName=WPO_WorldNoiseScale})
  n27 = MaterialFunctionCall(Float3=Multiply(A=@n10) {constB=0.22}) {materialFunction=&BreakOutFloat3Components}
  n28 = Add(A=@n27.R, B=@n27.R)
  n29 = Subtract(A=Time(), B=ScalarParameter() {parameterName=WPO_random_seed})
  n30 = Subtract(A=Subtract(A=Add(A=@n28, B=@n27.B)) {constB=2}, B=Frac(None=Multiply(A=@n29) {constB=0.6}))
  n31 = Multiply(A=Add(A=Sine(None=@n30))) {constB=0.5}
  n32 = Distance(A=CameraPositionWS(), B=WorldPosition().XYZ)
  n33 = ScalarParameter() {defaultValue=200, parameterName=Distance_Far}
  n34 = ScalarParameter() {defaultValue=100, parameterName=Distance_Near}
  n35 = Saturate(None=Divide(A=Subtract(A=@n32, B=@n33), B=Subtract(A=@n33, B=@n34)))
  Multiply(
   A=Multiply(A=VertexNormalWS(), B=Add(A=@n26, B=@n31))
   B=Multiply(A=OneMinus(None=@n35), B=ScalarParameter() {defaultValue=0.5, parameterName=WPO_FinalScale})
  )
  Rotator(Coordinate=@n19, Time=Constant())
  TextureSampleParameter2D() {parameterName=Normals, samplerType=SAMPLERTYPE_Normal, texture=/BattlePassS18/Items/Pickups/Textures/T_Rainbow_Ink_Drop_Small_N}
  VertexColor()
 ]
}
