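# Rainbow_Ink_Drop_Med_Cue is a SoundCue in the Battle Pass S18 pickup sounds, named for a medium Rainbow Ink drop.
# It has no sound nodes and a null root, so the cue as shown references no audio to play.
# It carries only playback settings: loot-collect sound class, weapon-drop attenuation, reduced volume and a priority.
# Attenuation and concurrency override values are also stored, the latter describing a single instance that blocks new ones.
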SoundCue /BattlePassS18/Items/Pickups/Sounds/Rainbow_Ink_Drop_Med_Cue
root=null
nodes []
properties {
 attenuationSettings=&Fort_Weapon_Drop_Att
 maxDistance=1300
 priority=0.9
 soundClassObject=&Fort_Loot_Collect
 volumeMultiplier=0.318
 attenuationOverrides {falloffDistance=1000, stereoSpread=512, attenuationShapeExtents {x=0}}
 concurrencyOverrides {maxCount=1, resolutionRule=PreventNew}
 concurrencySet [null]
}
