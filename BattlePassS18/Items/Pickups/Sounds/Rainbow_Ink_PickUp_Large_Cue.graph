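# Sound cue for the large Rainbow Ink pickup in Battle Pass Season 18, routed to the Fort_Loot_Collect sound class.
# A single wave player feeds Rainbow_Ink_Pickup_Large_01 straight to the output at reduced volume, with no randomization.
# Spatial falloff comes from the Rainbow_Ink_Att attenuation asset.
# Local attenuation and stop-oldest concurrency values are stored but not switched on as overrides, and the concurrency set holds only an empty entry.
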SoundCue /BattlePassS18/Items/Pickups/Sounds/Rainbow_Ink_PickUp_Large_Cue
root=WavePlayer_2
nodes [{id=WavePlayer_2, properties {soundWaveAssetPtr=&/Rainbow_Ink_Pickup_Large_01}}]
properties {
 attenuationSettings=&/Rainbow_Ink_Att
 maxDistance=2150
 soundClassObject=&Fort_Loot_Collect
 volumeMultiplier=0.498
 attenuationOverrides {falloffDistance=350, nonSpatializedRadiusStart=200, stereoSpread=512, attenuationShapeExtents {x=150}}
 concurrencyOverrides {maxCount=2, resolutionRule=StopOldest}
 concurrencySet [null]
}
